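# Unlit additive particle material that draws a wobbly, iridescent bubble on sprites, ribbons and mesh particles.
# Panned noise distorts UVs feeding a 2D sphere gradient, which shapes the mask, edge glow and mirrored reflection highlights.
# Three phase-offset sines over time, per-particle random, streak texture and mask create the rainbow color.
# Switches add particle color/alpha, spline thickening or camera-distance shrinking, and a world sphere mask, alongside camera depth fade.
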Material /BRCosmetics/Effects/Fort_Effects/Materials/Water/M_BubbleGPU
graph {
 properties {
  bAutomaticallySetUsageInEditor=false
  bUsedWithMeshParticles=true
  bUsedWithNiagaraRibbons=true
  bUsedWithNiagaraSprites=true
  bUsedWithParticleSprites=true
  bUsedWithStaticMesh=false
  blendMode=BLEND_Additive
  pixelDepthOffsetMode=PDOM_Legacy
  shadingModel=MSM_Unlit
  usedShadingModels=Unlit
 }
 nodes [
  n1 = Multiply(A=Time() {period=1}) {constB=0.1}
  n2 = ParticleRandom()
  n3 = Constant() {r=0.0125}
  n4 = MaterialFunctionCall(Float2=TextureCoordinate()) {materialFunction=&BreakOutFloat2Components}
  n5 = ScalarParameter() {defaultValue=1, parameterName=Base_U_Tile}
  n6 = ScalarParameter() {defaultValue=1, parameterName=Base_V_Tile}
  n7 = Add(A=TextureCoordinate()) {constB=-0.5}
  n8 = MaterialFunctionCall(Float2=TextureCoordinate()) {materialFunction=&BreakOutFloat2Components}
  n9 = ScalarParameter() {defaultValue=0.3, parameterName=Distort_U_Tile}
  n10 = ScalarParameter() {defaultValue=0.3, parameterName=Distort_V_Tile}
  n11 = Add(
   A=AppendVector(A=Multiply(A=@n8.R, B=@n9), B=Multiply(A=@n8.G, B=@n10))
   B=ShadingPathSwitch(Default=ParticleRandom(), Mobile=Constant()) {desc=Migrated from FeatureLevelSwitch}
  )
  n12 = AppendVector(
   A=ScalarParameter() {defaultValue=0.02, parameterName=Distort_U_PanSpeed}
   B=ScalarParameter() {defaultValue=0.15, parameterName=Distort_V_PanSpeed}
  )
  n13 = MaterialFunctionCall() {materialFunction=&LinearGradient}
  n14 = LinearInterpolate(
   B=TextureSample(UVs=Panner(Coordinate=@n11, Speed=@n12) {speedX=0.02, speedY=0.15}) {samplerType=SAMPLERTYPE_Grayscale, texture=&LowResBlurredNoise}.G
   Alpha=LinearInterpolate(A=SquareRoot(None=@n13.UGradient), Alpha=ScalarParameter() {defaultValue=1, parameterName=Lerp_GradientBlendDistortNoise})
  )
  n15 = Add(
   A=AppendVector(A=Multiply(A=@n4.R, B=@n5), B=Multiply(A=@n4.G, B=@n6))
   B=Multiply(A=Multiply(A=@n7, B=@n14), B=ScalarParameter() {defaultValue=4.408, parameterName=WibbStrength})
  )
  n16 = VectorParameter() {parameterName=SphereMaskPosition, defaultValue=<rgba 0.5 0.5 0 0>}
  n17 = MaterialFunctionCall!(
   UVs=@n15
   CenterPosition=LinearInterpolate(A=ComponentMask(None=@n16.RGB) {g=true, r=true}, B=Constant2Vector() {g=0.5, r=0.5}, Alpha=@n14)
   Radius=ScalarParameter() {defaultValue=0.5, parameterName=SphereMaskRadius}
  ) {materialFunction=/Engine/Functions/Engine_MaterialFunctions01/Gradient/SphereGradient-2D}
  n18 = OneMinus(None=@n17.Result 0-1)
  n19 = Multiply(A=Multiply(A=Add(A=@n15) {constB=-0.5}, B=LinearInterpolate(Alpha=@n18) {constA=0.1})) {constB=2}
  n20 = TextureSample(UVs=Panner(Coordinate=@n19) {speedY=0.2}) {texture=/Game/Effects/Fort_Effects/Textures/Energy/T_OutpostShockwaveStreaks_02}
  n21 = ScalarParameter() {defaultValue=1.1194, parameterName=Texture Strength}
  n22 = ScalarParameter() {defaultValue=1, parameterName=SphereMaskContrast}
  n23 = Multiply(A=Multiply(A=@n17.Result 0-1, B=Power(Base=@n18, Exp=@n22))) {constB=4}
  n24 = ScalarParameter() {defaultValue=1.1837, parameterName=Round Strength}
  n25 = Add(
   A=ShadingPathSwitch(Default=Add(A=@n1, B=@n2), Mobile=Add(A=@n1, B=@n3)) {desc=Migrated from FeatureLevelSwitch}
   B=Add(A=Multiply(A=@n20.G, B=@n21), B=Multiply(A=@n23, B=@n24))
  )
  n26 = Sine(None=Add(A=@n25) {constB=0})
  n27 = ScalarParameter() {defaultValue=0.2524, parameterName=SeperationGap}
  n28 = Sine(None=Add(A=@n25, B=@n27) {constB=0.1})
  n29 = Add(A=@n25, B=Multiply(A=@n27) {constB=2}) {constB=0.15}
  n30 = ConstantBiasScale(None=AppendVector(A=AppendVector(A=@n26, B=@n28), B=Sine(None=@n29)))
  n31 = Multiply(A=@n30, B=ScalarParameter() {defaultValue=0.2, parameterName=Color Strength})
  n32 = Add(A=@n19) {constB=0.5}
  n33 = TextureSample(UVs=@n32) {texture=/Game/Items/Art_noLOD/Textures/T_Items_MineralPowder_Glass_Reflection}
  n34 = OneMinus(None=@n32)
  n35 = Multiply(
   A=Multiply(A=TextureSample(UVs=@n34) {texture=&T_Items_MineralPowder_Glass_Reflection}.G, B=@n30)
   B=ScalarParameter() {defaultValue=0.5095, parameterName=Secondary Highlight Strength}
  )
  n36 = Multiply(A=Multiply(A=Add(A=@n33.G, B=@n35), B=@n23), B=ScalarParameter() {defaultValue=2, parameterName=Highlight Strength})
  n37 = Multiply(A=@n18, B=@n18)
  n38 = Multiply(A=@n37, B=@n37)
  n39 = Multiply(A=@n38, B=@n38)
  n40 = ScalarParameter() {defaultValue=2, parameterName=Edge Strength}
  n41 = Desaturation(None=Add(A=Add(A=@n31, B=@n36), B=Multiply(A=@n39, B=@n40)), Fraction=ScalarParameter() {defaultValue=-1, parameterName=Hilarity})
  n42 = ParticleColor()
  MaterialFunctionCall(Emissive=StaticSwitchParameter(True=Multiply(A=@n41, B=@n42.RGB), False=@n41) {parameterName=UseParticleColorToHandleOpacityAndColor}) {materialFunction=&MF_FX_EmissiveAdjustments}
  n43 = ScalarParameter() {defaultValue=5, parameterName=TrailWidth}
  n44 = ScalarParameter() {defaultValue=216.0992, parameterName=CamFadeDist}
  n45 = AppendVector(A=TextureCoordinate(), B=Constant())
  n46 = ParticleSize()
  n47 = Multiply(
   A=LinearInterpolate(Alpha=MaterialFunctionCall(Distance=@n44) {materialFunction=&Near_Camera_Fade}.Result) {constA=-1, constB=0}
   B=Multiply(A=Transform(None=@n45), B=ComponentMask(None=@n46) {r=true})
  )
  n48 = Constant()
  StaticSwitchParameter(
   True=MaterialFunctionCall(WidthBase=@n43, WidthTip=@n43, Expand U or V UV Channel=StaticBool()) {materialFunction=&SplineThicken}.WorldPositionOffset
   False=StaticSwitchParameter(True=ShadingPathSwitch(Default=@n47, Mobile=@n48) {desc=Migrated from FeatureLevelSwitch}, False=Constant()) {defaultValue=true, parameterName=ShrinkBasedOnCameraDistance}
  ) {parameterName=UseSplineThicken}
  n49 = MaterialFunctionCall(Float2=TextureCoordinate()) {materialFunction=&BreakOutFloat2Components}
  n50 = LinearInterpolate(A=@n42.A, B=Multiply(A=@n42.A, B=OneMinus(None=@n49.R)), Alpha=ScalarParameter() {parameterName=GradientOpacityLerp})
  n51 = Multiply(
   A=StaticSwitchParameter(True=Multiply(A=@n23, B=@n50), False=@n23) {parameterName=UseParticleColorToHandleOpacityAndColor}
   B=MaterialFunctionCall(
    Fade Length=ScalarParameter() {group=CameraDepthFade, parameterName=FadeLength}
    Fade Offset=ScalarParameter() {group=CameraDepthFade, parameterName=FadeOffset}
   ) {materialFunction=&CameraDepthFade}.Result
  )
  StaticSwitchParameter(
   True=Multiply(A=MaterialFunctionCall() {materialFunction=/Game/Athena/Items/Gameplay/SilkyVeteran/FX/SlurpVatMask}.Result, B=@n51)
   False=@n51
  ) {parameterName=Use World Sphere Mask}
 ]
}
