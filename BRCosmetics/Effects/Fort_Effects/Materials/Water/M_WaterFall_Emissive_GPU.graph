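# Translucent Default Lit particle sprite material, named for emissive GPU waterfall effects, using per-vertex directional volumetric lighting.
# Particle color RGB feeds the shared FX emissive adjustment function, alongside a sphere normal map sample.
# Particle alpha multiplies a round spark mask through a shading-path switch.
# The default path's speed-driven lerp blends the mask's red channel with itself, so speed has no effect and matches mobile.
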Material /BRCosmetics/Effects/Fort_Effects/Materials/Water/M_WaterFall_Emissive_GPU
graph {
 properties {
  bUsedWithParticleSprites=true
  blendMode=BLEND_Translucent
  pixelDepthOffsetMode=PDOM_Legacy
  translucencyLightingMode=TLM_VolumetricPerVertexDirectional
  usedShadingModels=Default Lit
 }
 nodes [
  n1 = ParticleColor()
  MaterialFunctionCall(Emissive=@n1.RGB) {materialFunction=&MF_FX_EmissiveAdjustments}
  n2 = TextureSample() {texture=/Game/Global/Textures/Default/Masks_and_Gradients/T_SparkRound}
  n3 = Clamp(None=Divide(A=ParticleSpeed(), B=ScalarParameter() {defaultValue=100, parameterName=ParticleSpeed}))
  Multiply(A=@n1.A, B=ShadingPathSwitch(Default=LinearInterpolate(A=@n2.R, B=@n2.R, Alpha=@n3), Mobile=@n2.R) {desc=Migrated from FeatureLevelSwitch})
  TextureSample() {samplerType=SAMPLERTYPE_Normal, texture=&T_Sphere_N}
 ]
}
